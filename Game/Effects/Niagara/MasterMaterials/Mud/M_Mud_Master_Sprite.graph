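# Master masked, two-sided, lit material for mud effects on Niagara sprites, ribbons and mesh particles.
# A radial flowmap, scaled by the CenterWarp dynamic parameter, distorts tiled UVs for splash normal and mask textures.
# Particle alpha and camera depth fade set the threshold that erodes mask channels and the normal's Z.
# It also includes Fresnel-shaped particle-color emissive, a jungle gameplay mud function and switchable mask-driven pixel depth offset.
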Material /Game/Effects/Niagara/MasterMaterials/Mud/M_Mud_Master_Sprite
graph {
 properties {
  bUsedWithNiagaraMeshParticles=true
  bUsedWithNiagaraRibbons=true
  bUsedWithNiagaraSprites=true
  blendMode=BLEND_Masked
  pixelDepthOffsetMode=PDOM_Legacy
  twoSided=true
  usedShadingModels=Default Lit
 }
 nodes [
  BreakMaterialAttributes!(Attr=MaterialFunctionCall!(FallBackMobileMudGrayscale) {materialFunction=/Game/Athena/Asteria/Environments/Landscape/Materials/MaterialFunctions/MF_JungleGameplayMud}.Material)
  n1 = VectorParameter() {parameterName=Tile, defaultValue=<rgba 1 1 0 0>}
  n2 = MaterialFunctionCall(Texture Scale=Constant() {r=1.4}) {materialFunction=&ScaleUVsByCenter}
  n3 = Subtract(A=ComponentMask(None=TextureSample(UVs=@n2.UVs) {samplerType=SAMPLERTYPE_LinearColor, texture=/Game/Environments/Hexmap/HexClouds/T_RadialFlowmap01}.RGB) {g=true, r=true}) {constB=0.5}
  n4 = DynamicParameter() {parameterIndex=1, defaultValue=<rgba 0 0 1 1>, paramNames [CenterWarp,MeshWPOBend,Param3,Param4]}
  n5 = Multiply(A=AppendVector(A=@n1.R, B=@n1.R), B=Add(A=TextureCoordinate(), B=Multiply(A=@n3, B=@n4.CenterWarp)))
  n6 = TextureSampleParameter2D(UVs=@n5) {parameterName=Normal, samplerType=SAMPLERTYPE_Normal, texture=/Game/Effects/Fort_Effects/Textures/Water/T_water_splash_04_n}
  n7 = ParticleColor()
  n8 = MaterialFunctionCall(
   Fade Length=ScalarParameter() {parameterName=Camera Depth Fade Length}
   Fade Offset=ScalarParameter() {parameterName=Camera Depth Fade Offset}
  ) {materialFunction=&CameraDepthFade}
  n9 = LinearInterpolate(Alpha=Saturate(None=Multiply(A=@n7.A, B=@n8.Result))) {constA=1, constB=0}
  n10 = Saturate(None=Divide(A=Subtract(A=@n6.B, B=@n9), B=OneMinus(None=@n9)))
  LinearInterpolate(
   A=Normalize(VectorInput=MaterialFunctionCall(X=@n6.R, Y=@n6.R, Z=@n10) {materialFunction=&MakeFloat3}.Result)
   B=Constant3Vector() {constant=<rgba 0 0 1 0>}
   Alpha=ScalarParameter() {parameterName=Flatten Normal}
  )
  n11 = OneMinus(None=Fresnel(ExponentIn=ScalarParameter() {defaultValue=0.5, parameterName=EmissiveFresnelExp}) {baseReflectFraction=0, exponent=7})
  MaterialFunctionCall(Emissive=Multiply(A=ParticleColor().RGB, B=Power(Base=@n11) {constExponent=3})) {materialFunction=&MF_FX_EmissiveAdjustments}
  n12 = TextureSampleParameter2D(UVs=@n5) {parameterName=Mask, texture=/Game/Effects/Fort_Effects/Textures/Water/T_water_splash_04}
  n13 = ChannelMaskParameter(None=@n12.RGB) {maskChannel=Blue, parameterName=ErodeMask, defaultValue=<rgba 0 0 1 0>}
  n14 = Subtract(A=@n13.R, B=@n9)
  n15 = Constant() {r=0.2}
  Multiply(A=Saturate(None=Divide(A=@n14, B=@n15)), B=ChannelMaskParameter(None=@n12.RGBA) {parameterName=OpacityMask}.R)
  n16 = OneMinus(None=@n7.A)
  Saturate(None=Divide(A=Subtract(A=@n13.R, B=@n16), B=ScalarParameter() {defaultValue=0.5, parameterName=Blend Alpha Softness}))
  ScalarParameter() {parameterName=Metallic}
  ScalarParameter() {defaultValue=0.6, parameterName=Specular}
  ScalarParameter() {defaultValue=0.6, parameterName=Roughness}
  StaticSwitchParameter(True=Multiply(A=OneMinus(None=@n12.R), B=ScalarParameter() {defaultValue=1, parameterName=PDO Scale}), False=Constant()) {parameterName=Use PDO}
  VectorParameter() {parameterName=Base Color, defaultValue=<rgba 0.3137 0.0824 0.0314 1>}
 ]
}
